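# Additive unlit particle material that draws an animated kaleidoscope glow, stored with the Viper backpack's glitch effects.
# Mirrored radial sine patterns drive curve-atlas colors, blended with a distorted layer off mobile, then modulate two-color chromatic striations.
# Particle color and time of day scale the emissive, while sphere masks hollow the center and soften edges, with depth fade off mobile.
# A position offset collapses vertices toward the actor origin as the camera nears.
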Material /Game/Accessories/FORT_Backpacks/Backpack_M_MED_Viper/FX/Viper_Glitch/M_Backpack_Haunt_Kaleidoscope
graph {
 properties {
  bUsedWithNiagaraSprites=true
  bUsedWithParticleSprites=true
  blendMode=BLEND_Additive
  pixelDepthOffsetMode=PDOM_Legacy
  shadingModel=MSM_Unlit
  usedShadingModels=Unlit
 }
 nodes [
  n1 = ActorPositionWS()
  n2 = Distance(A=CameraPositionWS(), B=@n1)
  n3 = ScalarParameter() {defaultValue=200, parameterName=NearShrinkDist}
  LinearInterpolate(A=Subtract(A=@n1, B=WorldPosition().XYZ), Alpha=Saturate(None=Divide(A=@n2, B=@n3))) {constB=0}
  n4 = Constant() {r=1}
  n5 = MaterialFunctionCall(Value_Morning=@n4, Value_Day=ScalarParameter() {defaultValue=1.5, parameterName=DayMult}, Value_Evening=@n4, Value_Night=@n4) {materialFunction=&Scalar_Time_of_Day}
  n6 = ParticleColor()
  n7 = Multiply(A=Time(), B=ScalarParameter() {defaultValue=0.15, parameterName=MasterSpeed})
  n8 = ScalarParameter() {defaultValue=0.035, parameterName=StriationWobbliness}
  n9 = AppendVector(A=TextureCoordinate(), B=Constant())
  n10 = MaterialFunctionCall() {materialFunction=&VectorToRadialValue}
  n11 = MaterialFunctionCall(Float2=@n10.Radial Coordinates) {materialFunction=&BreakOutFloat2Components}
  n12 = Sine(None=@n10.Vector Converted to Angle) {period=0.25}
  n13 = Multiply(A=@n7, B=VectorParameter() {parameterName=KaleidoscopePanSpeed, defaultValue=<rgba 0.04 0.08 0.16 0>}.RGB)
  n14 = Multiply(A=Add(A=Floor(None=@n12), B=Ceil(None=@n12)), B=@n13)
  n15 = Sine(None=Add(A=@n11.R, B=ComponentMask(None=@n14) {g=true, r=true})) {period=0.5}
  n16 = ConstantBiasScale(None=Sine(None=@n11.G) {period=1.26})
  n17 = ConstantBiasScale(None=Sine(None=@n11.G) {period=1.18})
  n18 = ComponentMask(None=@n13) {b=true}
  n19 = ConstantBiasScale(None=Sine(None=@n18) {period=0.87})
  n20 = AppendVector(
   A=Power(Base=Abs(None=@n15), Exp=ScalarParameter() {defaultValue=3, parameterName=PowerX})
   B=Power(Base=LinearInterpolate(A=@n16, B=@n17, Alpha=@n19), Exp=ScalarParameter() {defaultValue=1, parameterName=PowerY})
  )
  n21 = DotProduct(A=Normalize(VectorInput=@n9), B=Normalize(VectorInput=@n20))
  n22 = Constant() {r=0.5}
  n23 = SphereMask(A=TextureCoordinate(), B=@n22, Radius=@n22) {hardnessPercent=0}
  n24 = LinearInterpolate(
   A=ScalarParameter() {defaultValue=1.5, parameterName=KaleidoscopePowerOuter}
   B=ScalarParameter() {defaultValue=1.2, parameterName=KaleidoscopePower}
   Alpha=@n23
  ) {constA=5, constB=0.6}
  n25 = SphereMask(A=TextureCoordinate(), B=Constant() {r=0.5}) {attenuationRadius=1, hardnessPercent=0}
  n26 = Add(A=Power(Base=@n25), B=ScalarParameter() {parameterName=CurveOffset})
  n27 = CurveAtlasRowParameter(CurveTime=Multiply(A=Power(Base=@n21, Exp=@n24) {constExponent=3}, B=@n26)) {
   atlas=&AthenaCurveAtlas01
   curve=/Game/Athena/Environments/Nexus/Materials/FXshader/Curve_Color
   defaultValue=78
   parameterName=Param_1
  }
  n28 = ScalarParameter() {defaultValue=5, parameterName=Distortion}
  n29 = SphereMask(A=TextureCoordinate(), B=Constant() {r=0.5}, Radius=ScalarParameter() {defaultValue=1, parameterName=DistortHoleRadius}) {hardnessPercent=0}
  n30 = ScalarParameter() {defaultValue=9, parameterName=DistortionPower}
  n31 = Multiply(A=Multiply(A=Power(Base=@n29, Exp=@n30), B=Sine(None=@n10.Vector Converted to Angle) {period=0.5}), B=Subtract(A=@n9) {constB=0.5})
  n32 = Sine(None=@n18) {period=1.74}
  n33 = SphereMask(
   A=Add(A=Multiply(A=@n28, B=@n31), B=@n20)
   B=Constant() {r=0.5}
   Radius=LinearInterpolate(A=ScalarParameter() {defaultValue=0.9}, B=ScalarParameter() {defaultValue=1.4, parameterName=Param_2}, Alpha=ConstantBiasScale(None=@n32))
  ) {attenuationRadius=1, hardnessPercent=0}
  n34 = CurveAtlasRowParameter(CurveTime=Multiply(A=@n26, B=Power(Base=@n33, Exp=@n24) {constExponent=3})) {
   atlas=&AthenaCurveAtlas01
   curve=&Curve_Color
   defaultValue=78
   parameterName=Param_1
  }
  n35 = ShadingPathSwitch(Default=LinearInterpolate(A=@n27, B=@n34), Mobile=@n27) {desc=Migrated from FeatureLevelSwitch}
  n36 = Desaturation(None=@n35)
  n37 = Add(
   A=Multiply(A=@n7, B=ScalarParameter() {defaultValue=0.04, parameterName=StriationPanSpeed})
   B=Add(A=Multiply(A=@n8, B=@n36) {constB=0.2}, B=@n25)
  )
  n38 = Power(Base=ConstantBiasScale(None=Sine(None=@n37) {period=0.01}))
  n39 = LinearInterpolate(
   A=VectorParameter() {parameterName=Chroma1, defaultValue=<rgba 0 2 1.9646 0>}.RGB
   B=VectorParameter() {parameterName=Chroma2, defaultValue=<rgba 2 0 1.8202 0>}.RGB
   Alpha=@n38
  )
  n40 = Constant() {r=-1}
  n41 = LinearInterpolate(A=Multiply(A=Desaturation(None=@n39, Fraction=@n40), B=@n38), Alpha=Saturate(None=LinearInterpolate(Alpha=@n25) {constA=-0.5, constB=1.6})) {constB=0}
  n42 = Add(
   A=Multiply(A=@n35, B=ScalarParameter() {defaultValue=1, parameterName=KaleidoscopeColorMult})
   B=VectorParameter() {parameterName=BackgroundColor, defaultValue=<rgba 0.0474 0.0294 0.5 0>}.RGB
  )
  n43 = Multiply(A=@n6.RGB, B=Multiply(A=Add(A=@n41, B=@n42), B=@n36))
  n44 = Saturate(None=@n43)
  n45 = ScalarParameter() {defaultValue=1, parameterName=HDRMultEdge}
  n46 = ShadingPathSwitch(
   Default=Power(
    Base=MaterialFunctionCall(LDR Input=@n44, HDR Tint=@n43, HDR Intensity Multiplier=@n45) {materialFunction=&DeriveHDRfromLDR}.RemappedRange
    Exp=ScalarParameter() {defaultValue=1, parameterName=OverallPowerOuter}
   )
   Mobile=@n43
  ) {desc=Migrated from FeatureLevelSwitch}
  n47 = SphereMask(
   A=TextureCoordinate()
   B=Constant() {r=0.5}
   Radius=ScalarParameter() {defaultValue=0.4, parameterName=InnerHoleRadius}
   Hardness=ScalarParameter() {defaultValue=0.15, parameterName=InnerHoleHardness}
  )
  n48 = ScalarParameter() {defaultValue=2, parameterName=InnerHolePower}
  MaterialFunctionCall(Emissive=LinearInterpolate(A=Multiply(A=@n5.Result, B=@n46), Alpha=Power(Base=@n47, Exp=@n48) {constExponent=4}) {constB=0}) {materialFunction=&MF_FX_EmissiveAdjustments}
  n49 = Multiply(A=@n6.A, B=Power(Base=@n23, Exp=ScalarParameter() {defaultValue=3, parameterName=OpacityEdgePower}) {constExponent=4})
  ShadingPathSwitch(
   Default=DepthFade(Opacity=@n49, FadeDistance=ScalarParameter() {defaultValue=18, parameterName=FadeDistance}) {fadeDistanceDefault=50}
   Mobile=@n49
  ) {desc=Migrated from FeatureLevelSwitch}
 ]
}
