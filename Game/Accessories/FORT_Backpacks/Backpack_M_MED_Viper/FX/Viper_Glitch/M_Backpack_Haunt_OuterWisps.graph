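# Additive unlit particle material for the glowing outer wisps of the Viper backpack's glitch effect, usable on Niagara and particle sprites.
# Center-distorted radial UVs drive two offset panning turbulence samples, tinted for chromatic split, then multiplied by a sine-warped kaleidoscope mask colored from a curve atlas.
# Sphere masks cut an inner hole and soften edges, with HDR boost and depth fade off mobile.
# A camera-distance lerp shrinks geometry toward its pivot up close.
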Material /Game/Accessories/FORT_Backpacks/Backpack_M_MED_Viper/FX/Viper_Glitch/M_Backpack_Haunt_OuterWisps
graph {
 properties {
  bUsedWithNiagaraSprites=true
  bUsedWithParticleSprites=true
  blendMode=BLEND_Additive
  pixelDepthOffsetMode=PDOM_Legacy
  shadingModel=MSM_Unlit
  usedShadingModels=Unlit
 }
 nodes [
  n1 = MaterialFunctionCall() {materialFunction=&ObjectPivotPoint}
  n2 = Distance(A=CameraPositionWS(), B=@n1.Object Pivot Location)
  n3 = ScalarParameter() {defaultValue=200, parameterName=NearShrinkDist}
  LinearInterpolate(A=Subtract(A=@n1.Object Pivot Location, B=WorldPosition().XYZ), Alpha=Saturate(None=Divide(A=@n2, B=@n3))) {constB=0}
  n4 = ParticleColor()
  n5 = Multiply(A=Subtract(A=TextureCoordinate()) {constB=0.5}) {constB=2}
  n6 = Multiply(A=ScalarParameter() {defaultValue=0.01, parameterName=ChromaOffset}, B=Normalize(VectorInput=@n5)) {constB=0.05}
  n7 = ComponentMask(None=VectorParameter() {parameterName=TextureTile, defaultValue=<rgba 1 0.25 0 0>}.RGB) {g=true, r=true}
  n8 = Power(
   Base=SphereMask(
    A=TextureCoordinate()
    B=Constant() {r=0.5}
    Radius=ScalarParameter() {defaultValue=0.12, parameterName=DistortHoleRadius}
    Hardness=ScalarParameter() {defaultValue=0.12, parameterName=DistortHoleHardness}
   )
   Exp=ScalarParameter() {defaultValue=16, parameterName=DistortionPower}
  )
  n9 = Add(
   A=Multiply(A=Multiply(A=@n8, B=@n5), B=ScalarParameter() {defaultValue=50, parameterName=Distortion})
   B=MaterialFunctionCall() {materialFunction=&VectorToRadialValue}.Radial Coordinates
  )
  n10 = Constant2Vector() {g=-0.35}
  n11 = Constant() {r=0.5}
  n12 = Power(
   Base=SphereMask(A=TextureCoordinate(), B=@n11, Radius=@n11) {hardnessPercent=0}
   Exp=ScalarParameter() {defaultValue=3, parameterName=OpacityEdgePower}
  ) {constExponent=4}
  n13 = Multiply(A=Time(), B=ScalarParameter() {defaultValue=2, parameterName=MasterPanSpeed})
  n14 = Panner(Coordinate=Add(A=Multiply(A=@n7, B=@n9), B=Multiply(A=@n10, B=@n12) {constB=0.25}), Time=@n13) {speedX=-0.007, speedY=-0.02}
  n15 = TextureCoordinate()
  n16 = DDX(Value=@n15)
  n17 = DDY(Value=@n15)
  n18 = Multiply(
   A=TextureSample!(UVs=Add(A=@n6, B=@n14), Tex, "DDX(UVs)"=@n16, "DDY(UVs)"=@n17, Apply View MipBias) {mipValueMode=TMVM_Derivative, texture=/Game/Effects/Fort_Effects/Textures/Plasma/T_Turbulence_Stretched}.G
   B=VectorParameter() {parameterName=Chroma2, defaultValue=<rgba 0.6 0 0.3021 1>}.RGB
  )
  n19 = Multiply(
   A=TextureSample(UVs=@n14, "DDX(UVs)"=@n16, "DDY(UVs)"=@n17) {mipValueMode=TMVM_Derivative, texture=&T_Turbulence_Stretched}.G
   B=VectorParameter() {parameterName=Chroma1, defaultValue=<rgba 0 1.0185 1.5 1>}.RGB
  )
  n20 = TextureCoordinate()
  n21 = VectorParameter() {parameterName=MaskTile, defaultValue=<rgba 3 0.6 0 0>}
  n22 = VectorParameter() {parameterName=KaleidoscopePanSpeed, defaultValue=<rgba 0.03 -0.04 0 0>}
  n23 = Add(
   A=Multiply(A=@n9, B=ComponentMask(None=@n21.RGB) {g=true, r=true})
   B=Multiply(A=@n13, B=ComponentMask(None=@n22.RGB) {g=true, r=true}) {constB=0.25}
  )
  n24 = Add(A=Multiply(A=@n23, B=Constant2Vector() {g=0.7, r=-1})) {constB=0.25}
  n25 = LinearInterpolate(A=ConstantBiasScale(None=Sine(None=@n23) {period=0.5}), B=ConstantBiasScale(None=Sine(None=@n24)))
  n26 = Power(
   Base=SphereMask(A=Add(A=@n20, B=@n25), B=Constant() {r=0.5}, Radius=ScalarParameter() {defaultValue=1.4}) {attenuationRadius=1, hardnessPercent=0}
   Exp=ScalarParameter() {defaultValue=4, parameterName=KaleidoscopePower}
  ) {constExponent=3}
  n27 = CurveAtlasRowParameter(CurveTime=@n26) {
   atlas=&AthenaCurveAtlas01
   curve=/Game/Athena/Environments/Nexus/Materials/FXshader/Curve_Color
   defaultValue=78
   parameterName=Param_1
  }
  n28 = Multiply(A=@n27, B=ScalarParameter() {defaultValue=1.25, parameterName=KaleidoscopeColorMult})
  n29 = ScalarParameter() {defaultValue=1.25, parameterName=KaleidoscopeColorAdd}
  n30 = Multiply(A=Desaturation(None=LinearInterpolate(A=@n18, B=@n19), Fraction=Constant() {r=-0.8}), B=Multiply(A=@n26, B=Add(A=@n28, B=@n29) {constB=2}))
  n31 = ScalarParameter() {defaultValue=1, parameterName=HDRMultEdge}
  n32 = ScalarParameter() {defaultValue=1.5, parameterName=HDRMult}
  n33 = ShadingPathSwitch(
   Default=MaterialFunctionCall(LDR Input=Saturate(None=@n30), HDR Tint=@n30, HDR Intensity Multiplier=LinearInterpolate(A=@n31, B=@n32, Alpha=@n12) {constA=0.5}) {materialFunction=&DeriveHDRfromLDR}.RemappedRange
   Mobile=@n30
  ) {desc=Migrated from FeatureLevelSwitch}
  n34 = Power(Base=SphereMask(
    A=TextureCoordinate()
    B=Constant() {r=0.5}
    Radius=ScalarParameter() {defaultValue=0.088, parameterName=InnerHoleRadius}
    Hardness=ScalarParameter() {defaultValue=0.62, parameterName=InnerHoleHardness}
   )) {constExponent=4}
  MaterialFunctionCall(Emissive=Multiply(A=@n4.RGB, B=LinearInterpolate(A=@n33, Alpha=@n34) {constB=0})) {materialFunction=&MF_FX_EmissiveAdjustments}
  n35 = Multiply(A=@n4.A, B=@n12)
  ShadingPathSwitch(
   Default=DepthFade(Opacity=@n35, FadeDistance=ScalarParameter() {defaultValue=10, parameterName=FadeDistance}) {fadeDistanceDefault=25}
   Mobile=@n35
  ) {desc=Migrated from FeatureLevelSwitch}
 ]
}
