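# Translucent, two-sided unlit mesh-particle material for the Viper backpack's colorful glitch effect.
# Color blends panning noise and smoke with negatively desaturated sky and nebula textures, tinted by particle color.
# Alpha combines bump-offset turbulence with curve-atlas view-angle ramps, dimmed outside daytime.
# A "Backpack Viper Ray" switch instead uses vertex-green-masked Perlin alpha and adds a pulsing offset away from the particle center; "Hidden texture" zeroes alpha.
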Material /Game/Accessories/FORT_Backpacks/Backpack_M_MED_Viper/FX/Viper_Glitch/M_Backpack_Viper_Glitch_Colorful_Panner
graph {
 properties {
  bUsedWithMeshParticles=true
  bUsedWithNiagaraMeshParticles=true
  blendMode=BLEND_Translucent
  pixelDepthOffsetMode=PDOM_Legacy
  shadingModel=MSM_Unlit
  twoSided=true
  usedShadingModels=Unlit
 }
 nodes [
  n1 = ParticleColor()
  n2 = VertexColor()
  n3 = Power(Base=OneMinus(None=@n2.G)) {constExponent=10}
  n4 = Multiply(A=@n1.A, B=Multiply(A=TextureSample() {texture=/Game/Global/Textures/Default/Noise/3xPerlin_M}.R, B=Clamp(None=@n3)))
  n5 = Multiply(A=@n4) {constB=0.75}
  n6 = ParticleColor()
  n7 = ScalarParameter() {defaultValue=0.02, parameterName=FresnelAffectTexture}
  n8 = VertexNormalWS()
  n9 = CameraVectorWS()
  n10 = CurveAtlasRowParameter(CurveTime=Abs(None=DotProduct(A=@n8, B=@n9))) {
   atlas=&AthenaCurveAtlas01
   curve=/Game/Athena/Environments/Nexus/Materials/FXshader/Curve_Ramps
   defaultValue=79
  }
  n11 = TextureCoordinate() {uTiling=8, vTiling=5}
  n12 = TextureCoordinate() {uTiling=3}
  n13 = TextureSample(UVs=Panner(Coordinate=@n12) {speedX=0.2, speedY=0.2}) {samplerType=SAMPLERTYPE_LinearColor, texture=&T_NoisePack_LinearColor}
  n14 = TextureSample(UVs=Panner(Coordinate=@n12) {speedX=-0.2, speedY=-0.2}) {samplerType=SAMPLERTYPE_LinearColor, texture=&T_NoisePack_LinearColor}
  n15 = ComponentMask(None=TextureCoordinate() {uTiling=0.5, vTiling=0.5}) {g=true}
  n16 = MaterialFunctionCall(In=Abs(None=Sine(None=@n15)), Contrast=Constant() {r=0.1}) {materialFunction=&CheapContrast}
  n17 = Add(
   A=Multiply(A=Subtract(A=@n13.R, B=@n14.G), B=@n16.Result)
   B=Multiply(A=Subtract(A=@n10.R), B=MaterialFunctionCall() {materialFunction=&DebugTimeSine}.Result)
  )
  n18 = Multiply(A=@n16.Result, B=ScalarParameter() {defaultValue=0.25, parameterName=Hight})
  n19 = Add(B=BumpOffset(Coordinate=@n11, Height=@n17, HeightRatioInput=@n18))
  n20 = Multiply(A=@n19) {constB=1.618}
  n21 = Multiply(
   A=TextureSample(UVs=Panner(Coordinate=@n19) {speedX=0.1, speedY=0.3}) {texture=/Game/Global/Textures/Default/Caustics/T_Turbulence_Seamless}.R
   B=TextureSample(UVs=Panner(Coordinate=@n20) {speedX=0.1, speedY=0.3}) {samplerType=SAMPLERTYPE_LinearColor, texture=&T_NoisePack_LinearColor}.R
  )
  n22 = TextureSample(UVs=BumpOffset(Coordinate=@n11, Height=@n17, HeightRatioInput=@n18)) {texture=&T_Turbulence_Seamless}
  n23 = OneMinus(None=@n16.Result)
  n24 = Add(
   A=Add(B=Multiply(A=@n10.G, B=@n7))
   B=Multiply(A=LinearInterpolate(A=@n21, B=@n22.R, Alpha=@n23), B=ScalarParameter() {defaultValue=0.5, parameterName=NoiseScale})
  )
  n25 = StaticSwitchParameter(
   True=Constant()
   False=StaticSwitchParameter(
    True=MaterialFunctionCall(Color_Morning=@n5, Color_Day=@n4, Color_Evening=@n5, Color_Night=@n5) {materialFunction=&Color_Time_of_Day}.Result
    False=Multiply(A=@n6.A, B=@n24)
   ) {parameterName=Backpack Viper Ray}
  ) {parameterName=Hidden texture}
  n26 = Multiply(A=@n25) {constB=0.3}
  MaterialFunctionCall(Color_Morning=@n26, Color_Day=@n25, Color_Evening=@n26, Color_Night=@n26) {materialFunction=&Color_Time_of_Day}
  n27 = TextureSample(UVs=Panner() {speedX=0.1, speedY=-0.1}) {samplerType=SAMPLERTYPE_LinearColor, texture=&T_NoisePack_LinearColor}
  n28 = Multiply(A=Multiply(A=@n27.RGB, B=@n27.RGB))
  n29 = Multiply(A=TextureSample() {texture=&T_Inky_Smoke_Tile_Chromatic}.R, B=@n28)
  n30 = Panner() {speedX=0.2, speedY=-0.2}
  n31 = Panner() {speedX=-0.2, speedY=0.2}
  n32 = LinearInterpolate(
   A=Desaturation(None=TextureSample(UVs=@n30) {texture=/Game/Packages/Fortress_Sky/Materials/T_FORT_Sky}.RGB, Fraction=Constant() {r=-5})
   B=Desaturation(
    None=TextureSample(UVs=@n31) {texture=/Game/Characters/Player/Male/Medium/Bodies/M_Med_Soldier_04/Skins/Galaxy/Textures/T_FN_Nebula}.RGB
    Fraction=Constant() {r=-5}
   )
  )
  n33 = Constant() {r=1}
  n34 = Add(
   A=Desaturation(None=Multiply(A=@n29) {constB=5}, Fraction=Constant() {r=-1})
   B=Multiply(A=LinearInterpolate(A=@n28, B=@n32) {constAlpha=0.8}, B=StaticSwitchParameter(True=@n33, False=@n24) {parameterName=Backpack Viper Ray})
  )
  n35 = Desaturation(None=Multiply(A=@n34) {constB=10}, Fraction=Constant() {r=0.5})
  MaterialFunctionCall(Emissive=StaticSwitchParameter(True=Multiply(A=@n34, B=@n1.RGB), False=Multiply(A=@n6.RGB, B=@n35)) {parameterName=Backpack Viper Ray}) {materialFunction=&MF_FX_EmissiveAdjustments}
  Multiply(A=Fresnel() {exponent=3}, B=Fresnel() {baseReflectFraction=0, exponent=0.5})
  n36 = WorldPosition()
  n37 = ParticlePositionWS()
  n38 = Multiply(A=Normalize(VectorInput=Subtract(A=@n36.XYZ, B=@n37)), B=Multiply(A=@n2.G))
  n39 = Constant() {r=0.5}
  n40 = MaterialFunctionCall(Speed=Constant() {r=0.5}) {materialFunction=&DebugTimeSine}
  n41 = LinearInterpolate(
   A=LinearInterpolate(
    A=Multiply(A=@n38) {constB=5}
    B=Multiply(A=@n38) {constB=10}
    Alpha=MaterialFunctionCall(Speed=@n39) {materialFunction=&DebugTimeSine}.Result
   )
   B=LinearInterpolate(A=Multiply(A=@n38) {constB=5}, B=Multiply(A=@n38) {constB=7}, Alpha=OneMinus(None=@n40.Result))
  )
  n42 = TextureSample(UVs=Panner() {speedY=0.2}) {texture=&3xPerlin_M}
  StaticSwitchParameter(True=Multiply(A=Multiply(A=@n41, B=@n42.R)) {constB=7}, False=Constant()) {parameterName=Backpack Viper Ray}
 ]
}
